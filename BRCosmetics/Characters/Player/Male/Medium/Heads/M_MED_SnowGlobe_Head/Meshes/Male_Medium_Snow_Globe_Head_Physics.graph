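# Physics asset for the male medium snow globe head cosmetic, set up for secondary motion on the tree bones inside the head.
# A capsule body on head_base anchors a chain of four capsule bodies on dyn_tree_01 through dyn_tree_04, with shorter capsules on the last two.
# Each link is constrained to the one before it with narrow swing and tighter twist limits, so the chain can bend only slightly at each joint.
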PhysicsAsset /BRCosmetics/Characters/Player/Male/Medium/Heads/M_MED_SnowGlobe_Head/Meshes/Male_Medium_Snow_Globe_Head_Physics
bodies {
 dyn_tree_01 [{shapeType=Capsule, length=4, radius=2, center=<2 0 0>, rotation=<rot 90 0 0>}]
 dyn_tree_02 [{shapeType=Capsule, length=4, radius=2, center=<2 0 0>, rotation=<rot 90 0 0>}]
 dyn_tree_03 [{shapeType=Capsule, length=2, radius=2, center=<1 0 0>, rotation=<rot 90 0 0>}]
 dyn_tree_04 [{shapeType=Capsule, length=2, radius=2, center=<1 0 0>, rotation=<rot 90 0 0>}]
 head_base [{shapeType=Capsule, length=2, radius=2}]
}
constraints [
 {bone1Name=dyn_tree_01, bone2Name=head_base, swing1LimitDegrees=15, swing2LimitDegrees=15, twistLimitDegrees=5}
 {bone1Name=dyn_tree_02, bone2Name=dyn_tree_01, swing1LimitDegrees=15, swing2LimitDegrees=15, twistLimitDegrees=5}
 {bone1Name=dyn_tree_03, bone2Name=dyn_tree_02, swing1LimitDegrees=15, swing2LimitDegrees=15, twistLimitDegrees=5}
 {bone1Name=dyn_tree_04, bone2Name=dyn_tree_03, swing1LimitDegrees=15, swing2LimitDegrees=15, twistLimitDegrees=5}
]
